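# Master character material for the HighTower Wasabi male medium body, using material attributes with Default Lit and Subsurface Profile shading on skeletal meshes.
# The shared character PreFX function feeds an optional Silk layer, masked by a selectable channel of SkinFX_Mask, then an optional HoloFoil layer driven by pre-skinned position and normal.
# An EmissiveFresnel function follows, then shared PostFX, which also receives the PreFX AO on several inputs.
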Material /BRCosmetics/Characters/Player/Male/Medium/Bodies/M_MED_HighTower_Wasabi_Blue/Materials/M_MED_HighTower_Wasabi_MASTER
graph {
 properties {
  bAutomaticallySetUsageInEditor=false
  bUseMaterialAttributes=true
  bUsedWithMeshDeformer=true
  bUsedWithMorphTargets=true
  bUsedWithSkeletalMesh=true
  pixelDepthOffsetMode=PDOM_Legacy
  shadingModel=MSM_FromMaterialExpression
  subsurfaceProfile=&SS_HeroSkin_02
  usedShadingModels=Default Lit | Subsurface Profile
 }
 nodes [
  n1 = MaterialFunctionCall() {materialFunction=&BaseFN_CharacterFunction_PreFX}
  n2 = TextureSampleParameter2D() {parameterName=SkinFX_Mask}
  n3 = StaticSwitchParameter(
   True=MaterialFunctionCall!(Alpha=ChannelMaskParameter(None=@n2.RGB) {group=Skin FX, parameterName=Silk Channel}.R, Material=@n1.Result) {materialFunction=/Game/Weapons/WeaponSkins/MaterialFunction/MF_Silk}.Result
   False=@n1.Result
  ) {group=Skin FX, parameterName=Use Silk}
  n4 = MaterialFunctionCall!(
   PreskinnedLocalPosition=VertexInterpolator(VS=PreSkinnedPosition()).PS
   PreskinnedLocalNormal=VertexInterpolator(VS=PreSkinnedNormal()).PS
   Material=@n3
  ) {materialFunction=/Game/Characters/Player/Female/Medium/Bodies/F_MED_HighTower_Squash/Materials/MF_HoloFoil}
  MaterialFunctionCall(
   AO=@n1.AO
   Distance Blend=@n1.AO
   Fresnel=@n1.AO
   Gmap=@n1.AO
   Material=MaterialFunctionCall!(In=StaticSwitchParameter(True=@n4.Result, False=@n3) {parameterName=BUseHoloFoil}) {materialFunction=/Game/Weapons/WeaponSkins/MaterialFunction/MF_EmissiveFresnel}.Result
  ) {materialFunction=&BaseFN_CharacterFunction_PostFX}
 ]
}
